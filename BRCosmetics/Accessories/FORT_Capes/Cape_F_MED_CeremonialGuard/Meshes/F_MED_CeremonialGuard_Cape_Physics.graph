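# Physics asset for the CeremonialGuard medium female cape cosmetic, giving its cape and strap bones capsule bodies for physical motion.
# Four large cape capsules form a constrained chain from the tiny left master body, with asymmetric swing and tight twist limits.
# Two smaller strap capsules chain from the central master body with uniformly tight limits.
# A right master body exists, but no constraints or chain bodies attach to it.
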PhysicsAsset /BRCosmetics/Accessories/FORT_Capes/Cape_F_MED_CeremonialGuard/Meshes/F_MED_CeremonialGuard_Cape_Physics
bodies {
 dyn_cape_MSTR [{shapeType=Capsule, length=0.505, radius=0.505, shapeName=dyn_cape_r_MSTR_capsule, rotation=<rot 0 -90 0>}]
 dyn_cape_l_01 [{shapeType=Capsule, length=8.2556, radius=9.4773, shapeName=dyn_cape_01_capsule, center=<7.148 -4.8083 7.2771>, rotation=<rot 86.4467 0 0>}]
 dyn_cape_l_02 [{
   shapeType=Capsule
   length=9.102
   radius=8.6308
   shapeName=dyn_cape_01_capsule
   center=<7.1979 -6.1353 5.1624>
   rotation=<rot 85.3026 180 -180>
  }]
 dyn_cape_l_03 [{shapeType=Capsule, length=9.102, radius=8.6308, shapeName=dyn_cape_01_capsule, center=<7.15 -5.0837 4.0762>, rotation=<rot 90 0 0>}]
 dyn_cape_l_04 [{shapeType=Capsule, length=9.102, radius=8.6308, shapeName=dyn_cape_01_capsule, center=<7.15 -2.0348 5.1365>, rotation=<rot 90 0 0>}]
 dyn_cape_l_MSTR [{shapeType=Capsule, length=0.505, radius=0.505, shapeName=dyn_cape_r_MSTR_capsule, rotation=<rot 0 -90 0>}]
 dyn_cape_r_MSTR [{shapeType=Capsule, length=0.505, radius=0.505, shapeName=dyn_cape_r_MSTR_capsule, rotation=<rot 0 -90 0>}]
 strap_01_joint [{
   shapeType=Capsule
   length=0.7985
   radius=1.7831
   shapeName=strap_01_joint_capsule
   center=<2.0372 -0.9449 0.1543>
   rotation=<rot 2.3365 -90.0035 -89.1413>
  }]
 strap_02_joint [{
   shapeType=Capsule
   length=1.8346
   radius=2.0067
   shapeName=strap_02_joint_capsule
   center=<2.6691 -0.3867 -0.0209>
   rotation=<rot 10.1086 -90 -88.3351>
  }]
}
constraints [
 {bone1Name=dyn_cape_l_01, bone2Name=dyn_cape_l_MSTR, swing1LimitDegrees=25, swing2LimitDegrees=55, twistLimitDegrees=5}
 {bone1Name=dyn_cape_l_02, bone2Name=dyn_cape_l_01, swing1LimitDegrees=25, swing2LimitDegrees=55, twistLimitDegrees=5}
 {bone1Name=dyn_cape_l_03, bone2Name=dyn_cape_l_02, swing1LimitDegrees=25, swing2LimitDegrees=55, twistLimitDegrees=5}
 {bone1Name=dyn_cape_l_04, bone2Name=dyn_cape_l_03, swing1LimitDegrees=25, swing2LimitDegrees=55, twistLimitDegrees=5}
 {bone1Name=strap_01_joint, bone2Name=dyn_cape_MSTR, swing1LimitDegrees=5, swing2LimitDegrees=5, twistLimitDegrees=5}
 {bone1Name=strap_02_joint, bone2Name=strap_01_joint, swing1LimitDegrees=5, swing2LimitDegrees=5, twistLimitDegrees=5}
]
