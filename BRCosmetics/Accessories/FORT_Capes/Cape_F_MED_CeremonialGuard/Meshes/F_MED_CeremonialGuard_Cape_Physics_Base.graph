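# Physics asset for the F_MED Ceremonial Guard cape cosmetic, covering the cape's master bone and two strap joints.
# Each bone has one capsule body, with the zero-length capsule on dyn_cape_MSTR acting as a sphere.
# Constraints chain strap_02_joint to strap_01_joint to dyn_cape_MSTR, with tight angular limits that give a small swing on one axis, none on the other and minimal twist.
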PhysicsAsset /BRCosmetics/Accessories/FORT_Capes/Cape_F_MED_CeremonialGuard/Meshes/F_MED_CeremonialGuard_Cape_Physics_Base
bodies {
 dyn_cape_MSTR [{shapeType=Capsule, length=0, radius=4, shapeName=dyn_cape_MSTR_capsule}]
 strap_01_joint [{
   shapeType=Capsule
   length=3.5909
   radius=4.5
   shapeName=strap_01_joint_capsule
   center=<2.9625 -4.1379 0.2276>
   rotation=<rot 2.342 -90 1.531>
  }]
 strap_02_joint [{
   shapeType=Capsule
   length=2.8193
   radius=5.5
   shapeName=strap_02_joint_capsule
   center=<3.3386 -6.3066 0.4339>
   rotation=<rot 4.1593 -90 -1.0543>
  }]
}
constraints [
 {bone1Name=strap_01_joint, bone2Name=dyn_cape_MSTR, swing1LimitDegrees=0, swing2LimitDegrees=5, twistLimitDegrees=1}
 {bone1Name=strap_02_joint, bone2Name=strap_01_joint, swing1LimitDegrees=0, swing2LimitDegrees=5, twistLimitDegrees=1}
]
